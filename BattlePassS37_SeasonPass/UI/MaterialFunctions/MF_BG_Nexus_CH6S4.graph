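# UI material function that builds the animated "Nexus" background for the Season 37 Battle Pass (CH6S4).
# A radial gradient base, three panning radial-masked cloud layers, funnel-UV noise light rays, and an MSDF icon with spotlight form the backdrop.
# Vignette, random noise, confetti-function embers and a panning-noise fire gradient add atmosphere.
# Named reroutes link the layers, with vault-world helpers supplying fit/fill UVs and focal point, and tunables exposed as grouped parameters.
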MaterialFunction /BattlePassS37_SeasonPass/UI/MaterialFunctions/MF_BG_Nexus_CH6S4
graph {
 properties {previewMaterialDomain=MD_UI}
 nodes [
  FunctionOutput(None=NamedRerouteUsage())
  n1 = NamedRerouteUsage()
  n2 = ScalarParameter() {defaultValue=1.2, group=MiddleClouds, parameterName=Top_MCloudSize, sliderMax=10}
  n3 = Constant2Vector() {g=0.125, r=0.07}
  n4 = ScalarParameter() {defaultValue=0.2, group=MiddleClouds, parameterName=Top_MCloudSpeedMult, sliderMax=1}
  n5 = TextureSample(UVs=Panner(
    Coordinate=MaterialFunctionCall(UVs=@n1, Texture Scale=@n2) {materialFunction=&ScaleUVsByCenter}.UVs
    Speed=Multiply(A=@n3, B=@n4)
   ) {speedX=0.125, speedY=0.125}) {texture=/BattlePassS37_SeasonPass/UI/Textures/Background/T_Clouds}
  NamedRerouteDeclaration(None=@n5.R) {name=M Cloud R Pan Top, nodeColor=<rgba 0 0 0 1>}
  n6 = VectorParameter() {parameterName=BackgroundLight, defaultValue=<rgba 0.0742 0.2423 0.6867 1>}
  n7 = MaterialFunctionCall(
   UV=MaterialFunctionCall(UVs=NamedRerouteUsage(), Texture Scale=ScalarParameter() {defaultValue=1.25, group=Background, parameterName=RadialSize}) {materialFunction=&ScaleUVsByCenter}.UVs
   GradientStart=ScalarParameter() {group=Background, parameterName=RadGradStart}
   GradientEnd=ScalarParameter() {defaultValue=1, group=Background, parameterName=RadGradEnd}
  ) {materialFunction=&MF_UI_RadialGradient}
  NamedRerouteDeclaration(None=LinearInterpolate(
    A=VectorParameter() {parameterName=BackgroundDark, defaultValue=<rgba 0.0033 0.0086 0.0262 1>}.RGB
    B=NamedRerouteDeclaration(None=@n6.RGB) {name=ColorBackgroundLight, nodeColor=<rgba 0 0 0 1>}
    Alpha=MaterialFunctionCall("Value (0-1)"=@n7.Result) {materialFunction=&MF_UI_EaseCurves}.EaseInOutSine
   )) {name=RadialBackgroundBase, nodeColor=<rgba 0 0 0 1>}
  n8 = TextureSample(UVs=MaterialFunctionCall(UVs=NamedRerouteUsage(), Texture Scale=ScalarParameter() {defaultValue=0.9, group=Icon, parameterName=IconScale, sliderMax=2}) {materialFunction=&ScaleUVsByCenter}.UVs) {texture=/BattlePassS37_SeasonPass/UI/Textures/Background/T_UI_BP37_OXRIcon}
  NamedRerouteDeclaration(None=SmoothStep(
    Min=ScalarParameter() {defaultValue=0.22, group=Icon, parameterName=IconSmoothMin, sliderMax=1}
    Max=ScalarParameter() {defaultValue=0.23, group=Icon, parameterName=IconSmoothMax, sliderMax=1}
    Value=MaterialFunctionCall!(MSDF RGB=@n8.RGB) {materialFunction=/FortUILibrary/MaterialFunctions/Shapes/Modifiers/MF_UI_MSDFtoSDF}.SDF
   )) {name=IconMask, nodeColor=<rgba 0 0 0 1>}
  n9 = MaterialFunctionCall!(Use Local UV, Texture Size, Scale Pivot Point) {materialFunction=/Game/Environments/FrontEnd/Blueprints/ItemPreview/VaultWorldTransitions/MaterialFunctions/MF_UI_VaultWorld_UV}
  NamedRerouteDeclaration(None=@n9.Scaled to Fit) {name=ScaleToFitUV, nodeColor=<rgba 0 0 0 1>}
  n10 = ComponentMask(None=MaterialFunctionCall() {materialFunction=&GetUserInterfaceUV}.Normalized UV) {g=true}
  n11 = ScalarParameter() {defaultValue=80, parameterName=L1-Pattern-X}
  n12 = ScalarParameter() {defaultValue=1, parameterName=L1-Pattern-Y}
  n13 = ScalarParameter() {defaultValue=6, parameterName=L1-Confetti-Local-Tile-X}
  n14 = ScalarParameter() {defaultValue=160, parameterName=L1-Confetti-Local-Tile-Y}
  n15 = ScalarParameter() {defaultValue=0.7, parameterName=L1-SmoothStep-Max}
  n16 = ScalarParameter() {defaultValue=0.5, parameterName=L1-SmoothStep-Min}
  n17 = Subtract(A=ComponentMask(None=NamedRerouteUsage()) {g=true}, B=ScalarParameter() {defaultValue=0.55, parameterName=L1-Gradient-Mask})
  n18 = ScalarParameter() {defaultValue=0.4, parameterName=L1-Gradient-Falloff}
  n19 = Multiply(
   A=MaterialFunctionCall!(
    "UV(V2)"=NamedRerouteUsage()
    Global-Angle=LinearInterpolate(Alpha=@n10) {constA=-1.1, constB=-1}
    "Pattern-Tiles(V2)"=AppendVector(A=@n11, B=@n12)
    RandomSeed=Constant() {r=123}
    Confetti-Speed-TimeFactor=ScalarParameter() {defaultValue=0.1, parameterName=L1-Confetti-Speed-TimeFactor}
    Confetti-Rotation-TimeFactor=ScalarParameter() {defaultValue=0.2, parameterName=L1-Confetti-Rotation-TimeFactor}
    Confetti-Local-Rotation-Range=ScalarParameter() {defaultValue=0.0075, parameterName=L1-Confetti-Local-Rotation-Range}
    "Confetti-Local-Tiles(V2)"=AppendVector(A=@n13, B=@n14)
    "SmoothStepRanges(V2)"=AppendVector(A=@n15, B=@n16)
   ) {materialFunction=/Game/UI/Frontend/ItemShop/Art/Cactus/MF_UI_Confetti}.Confetti-Mask
   B=Saturate(None=Multiply(A=@n17, B=@n18))
  )
  n20 = ScalarParameter() {defaultValue=3, parameterName=L1-Intensity}
  NamedRerouteDeclaration(None=Multiply(A=Constant3Vector() {constant=<rgba 1 0.3366 0.0015 1>}, B=Multiply(A=@n19, B=@n20))) {name=Embers, nodeColor=<rgba 0 0 0 1>}
  n21 = MaterialFunctionCall!(
   Vault World Color
   Focal Color Override
   Focal Scale Override=ScalarParameter() {defaultValue=1.8, group=Background, parameterName=FocalScale}
   Mask Blend
  ) {materialFunction=/Game/Environments/FrontEnd/Blueprints/ItemPreview/VaultWorldTransitions/MaterialFunctions/MF_UI_VaultWorld_FocalPoint}
  NamedRerouteDeclaration(None=@n21.Focal Mask) {name=FocalMask, nodeColor=<rgba 0 0 0 1>}
  NamedRerouteDeclaration(None=@n5.A) {name=M Cloud A Pan Top, nodeColor=<rgba 0 0 0 1>}
  n22 = NamedRerouteUsage()
  n23 = ScalarParameter() {defaultValue=1.2, group=MiddleClouds, parameterName=Mid_MCloudSize, sliderMax=10}
  n24 = Constant2Vector() {g=0.125, r=-0.125}
  n25 = ScalarParameter() {defaultValue=0.2, group=MiddleClouds, parameterName=Mid_MCloudSpeedMult, sliderMax=1}
  n26 = TextureSample(UVs=Panner(
    Coordinate=MaterialFunctionCall(UVs=@n22, Texture Scale=@n23) {materialFunction=&ScaleUVsByCenter}.UVs
    Speed=Multiply(A=@n24, B=@n25)
   ) {speedX=-0.125, speedY=0.125}) {texture=&T_Clouds}
  NamedRerouteDeclaration(None=@n26.R) {name=M Cloud R Mid, nodeColor=<rgba 0 0 0 1>}
  NamedRerouteDeclaration(None=@n26.A) {name=M Cloud A Mid, nodeColor=<rgba 0 0 0 1>}
  n27 = Multiply(
   A=MaterialFunctionCall!(UV=NamedRerouteUsage(), Seed) {materialFunction=&MF_UI_Random}.Result
   B=ScalarParameter() {defaultValue=0.1, group=Background, parameterName=NoiseMult, sliderMax=1}
  ) {constB=0.1}
  n28 = NamedRerouteUsage()
  n29 = LinearInterpolate(B=@n28, Alpha=NamedRerouteUsage())
  n30 = NamedRerouteUsage()
  n31 = Multiply(
   A=Add(A=NamedRerouteUsage(), B=NamedRerouteUsage())
   B=ScalarParameter() {defaultValue=0.6, group=MiddleClouds, parameterName=Top_MCloudDensityMult}
  )
  n32 = Abs(None=Subtract(A=Min(A=@n30, B=@n31), B=NamedRerouteUsage()))
  n33 = Multiply(
   A=VectorParameter() {group=MiddleClouds, parameterName=TopMColor, defaultValue=<rgba 0.0467 0.2831 0.9647 1>}.RGB
   B=LinearInterpolate(
    A=NamedRerouteUsage()
    B=NamedRerouteDeclaration(None=@n32) {name=TopMCloudMask, nodeColor=<rgba 0 0 0 1>}
    Alpha=ScalarParameter() {defaultValue=1, group=MiddleClouds, parameterName=TopMCloudOpacity}
   )
  )
  n34 = Multiply(
   A=Add(A=NamedRerouteUsage(), B=NamedRerouteUsage())
   B=ScalarParameter() {defaultValue=0.6, group=MiddleClouds, parameterName=Mid_MCloudDensityMult}
  )
  n35 = MaterialFunctionCall(UV=MaterialFunctionCall(UVs=NamedRerouteUsage(), Texture Scale=ScalarParameter() {defaultValue=3.7, group=MiddleClouds, parameterName=MidMCloudScale, sliderMax=10}) {materialFunction=&ScaleUVsByCenter}.UVs) {materialFunction=&MF_UI_RadialGradient}
  n36 = Abs(None=Subtract(A=Min(A=@n34, B=@n35.Result), B=@n35.Result))
  n37 = Multiply(
   A=VectorParameter() {group=MiddleClouds, parameterName=MidMColor, defaultValue=<rgba 0.004 0.0232 0.0762 1>}.RGB
   B=LinearInterpolate(
    A=NamedRerouteUsage()
    B=NamedRerouteDeclaration(None=@n36) {name=MidMCloudMask, nodeColor=<rgba 0 0 0 1>}
    Alpha=ScalarParameter() {defaultValue=1, group=MiddleClouds, parameterName=MidMCloudOpacity}
   ) {constAlpha=0.3}
  )
  n38 = Multiply(
   A=Add(A=NamedRerouteUsage(), B=NamedRerouteUsage())
   B=ScalarParameter() {defaultValue=0.6, group=MiddleClouds, parameterName=Bot_MCloudDensityMult}
  )
  n39 = MaterialFunctionCall(UV=MaterialFunctionCall(UVs=NamedRerouteUsage(), Texture Scale=ScalarParameter() {defaultValue=6.5, group=MiddleClouds, parameterName=BotMCloudScale, sliderMax=10}) {materialFunction=&ScaleUVsByCenter}.UVs) {materialFunction=&MF_UI_RadialGradient}
  n40 = Abs(None=Subtract(A=Min(A=@n38, B=@n39.Result), B=@n39.Result))
  n41 = Multiply(
   A=VectorParameter() {group=MiddleClouds, parameterName=BotMColor, defaultValue=<rgba 0.011 0.0168 0.0319 1>}.RGB
   B=LinearInterpolate(
    A=NamedRerouteUsage()
    B=NamedRerouteDeclaration(None=@n40) {name=BotMCloudMask, nodeColor=<rgba 0 0 0 1>}
    Alpha=ScalarParameter() {defaultValue=1, group=MiddleClouds, parameterName=BotMCloudOpacity}
   ) {constAlpha=0.2}
  )
  n42 = Add(A=Saturate(None=Add(A=@n33, B=@n37)), B=Saturate(None=Add(A=@n37, B=@n41)))
  n43 = MaterialFunctionCall!(Blend=LinearInterpolate(B=Saturate(None=@n42)) {constAlpha=1}, Base=NamedRerouteUsage()) {materialFunction=/Engine/Functions/Engine_MaterialFunctions03/Blends/Blend_Screen}
  n44 = NamedRerouteUsage()
  n45 = Saturate(None=Multiply(A=NamedRerouteUsage()) {constB=7.5})
  n46 = Multiply(A=Multiply(
    A=VectorParameter() {group=Icon, parameterName=IconColor, defaultValue=<rgba 0.3712 0.7529 0.8714 1>}.RGB
    B=LinearInterpolate(B=@n44, Alpha=@n45)
   ))
  n47 = MaterialFunctionCall(UV=MaterialFunctionCall(UVs=NamedRerouteUsage(), Texture Scale=ScalarParameter() {defaultValue=0.55, group=Icon, parameterName=IconSpotlightScale}) {materialFunction=&ScaleUVsByCenter}.UVs) {materialFunction=&MF_UI_RadialGradient}
  n48 = ScalarParameter() {defaultValue=0.7, group=Icon, parameterName=IconSpotlightOpacity, sliderMax=1}
  n49 = Add(
   A=Saturate(None=MaterialFunctionCall!(Base=@n43.Result, Add=@n46) {materialFunction=/Engine/Functions/Engine_MaterialFunctions02/Math/MultiplyAdd}.Result)
   B=Multiply(A=Multiply(A=@n47.Result, B=@n48), B=VectorParameter() {parameterName=BloomColor, defaultValue=<rgba 0.2086 0.5972 0.7758 1>}.RGB)
  )
  n50 = Saturate(None=MaterialFunctionCall(Blend=LinearInterpolate(A=@n28, B=@n29), Base=Saturate(None=@n49)) {materialFunction=&Blend_Screen}.Result)
  n51 = VectorParameter() {group=Vignette, parameterName=VignetteColor, defaultValue=<rgba 0.002 0.0046 0.0103 1>}
  n52 = MaterialFunctionCall(UVs=NamedRerouteUsage(), Texture Scale=ScalarParameter() {defaultValue=2, group=Vignette, parameterName=VignetteScale}) {materialFunction=&ScaleUVsByCenter}
  n53 = NamedRerouteDeclaration(None=MaterialFunctionCall("Value (0-1)"=MaterialFunctionCall(UV=@n52.UVs) {materialFunction=&MF_UI_RadialGradient}.Result) {materialFunction=&MF_UI_EaseCurves}.EaseInSine) {name=VignetteMask, nodeColor=<rgba 0 0 0 1>}
  n54 = LinearInterpolate(
   A=@n50
   B=Multiply(A=@n50, B=LinearInterpolate(A=@n51.RGB, Alpha=@n53))
   Alpha=ScalarParameter() {defaultValue=0.6, group=Vignette, parameterName=VignetteStrength, sliderMax=1}
  )
  n55 = Add(
   A=NamedRerouteUsage()
   B=LinearInterpolate(
    A=MaterialFunctionCall!(Blend=@n27, Base=@n54) {materialFunction=/Engine/Functions/Engine_MaterialFunctions03/Blends/Blend_ColorDodge}.Result
    B=@n54
   ) {constAlpha=0.8}
  )
  n56 = Constant3Vector() {constant=<rgba 1 0.2274 0 0>}
  n57 = Constant3Vector() {constant=<rgba 0.1198 0.0092 0 1>}
  n58 = NamedRerouteUsage()
  n59 = AppendVector(A=Constant(), B=ScalarParameter() {defaultValue=-0.05, group=Fire, parameterName=FireYTransform, sliderMax=1, sliderMin=-1})
  n60 = AppendVector(
   A=ScalarParameter() {defaultValue=1, group=Fire, parameterName=Radial-X-Tile2}
   B=ScalarParameter() {defaultValue=1, group=Fire, parameterName=Radial-Y-Tile2}
  )
  n61 = Multiply(A=NamedRerouteDeclaration(None=Add(A=@n58, B=@n59)) {name=FireUVs, nodeColor=<rgba 0 0 0 1>}, B=@n60)
  n62 = Multiply(A=@n60, B=Constant2Vector() {g=1, r=0.5})
  n63 = Saturate(None=Power(Base=Distance(A=@n61, B=@n62), Exp=ScalarParameter() {defaultValue=0.6, group=Fire, parameterName=Distance Powerfire}))
  n64 = Time()
  n65 = ScalarParameter() {defaultValue=0.2, group=Fire, parameterName=Speed}
  n66 = TextureSample(UVs=Panner(Coordinate=NamedRerouteUsage(), Time=Multiply(A=@n64, B=@n65)) {speedY=0.25}) {samplerType=SAMPLERTYPE_Grayscale, texture=/Game/UI/Frontend/ItemShop/Art/Rumble/LowResBlurredNoise_Rumble}
  n67 = Time()
  n68 = ScalarParameter() {defaultValue=0.1, group=Fire, parameterName=Speed2}
  n69 = TextureSample(UVs=Panner(Coordinate=NamedRerouteUsage(), Time=Multiply(A=@n67, B=@n68)) {speedY=0.25}) {samplerType=SAMPLERTYPE_Grayscale, texture=&LowResBlurredNoise_Rumble}
  n70 = ComponentMask(None=NamedRerouteUsage()) {g=true}
  n71 = ScalarParameter() {defaultValue=-0.067, group=Fire, parameterName=Subtract3}
  n72 = Multiply(
   A=Multiply(A=Subtract(A=@n70, B=@n71), B=ScalarParameter() {defaultValue=0.01, group=Fire, parameterName=DropOff3})
   B=ScalarParameter() {defaultValue=7.6175, group=Fire, parameterName=Intensity3}
  )
  n73 = Time()
  n74 = ScalarParameter() {defaultValue=1, group=Fire, parameterName=TopMaskScale}
  n75 = TextureSample(UVs=Panner(Coordinate=NamedRerouteUsage(), Time=Multiply(A=@n73, B=@n74)) {speedX=0.015, speedY=0.15}) {samplerType=SAMPLERTYPE_Grayscale, texture=&LowResBlurredNoise_Rumble}
  n76 = LinearInterpolate(
   B=Multiply(A=LinearInterpolate(A=@n56, B=@n57, Alpha=@n63), B=Add(A=@n66.RGB, B=@n69.RGB))
   Alpha=LinearInterpolate(A=Add(A=@n72, B=@n75.RGB))
  )
  n77 = Subtract(A=ComponentMask(None=NamedRerouteUsage()) {g=true}, B=ScalarParameter() {defaultValue=0.8, group=Fire, parameterName=SubtractYellow})
  n78 = ScalarParameter() {defaultValue=2, group=Fire, parameterName=FalloffYellow}
  n79 = LinearInterpolate(B=Constant3Vector() {constant=<rgba 1 0.6749 0 0>}, Alpha=Saturate(None=Multiply(A=@n77, B=@n78)))
  NamedRerouteDeclaration(None=LinearInterpolate(A=Saturate(None=@n55), B=Multiply(A=@n76, B=@n79), Alpha=@n79)) {name=MiddleCloud, nodeColor=<rgba 0 0 0 1>}
  n80 = NamedRerouteUsage()
  n81 = ScalarParameter() {defaultValue=1.2, group=MiddleClouds, parameterName=Bot_MCloudSize, sliderMax=10}
  n82 = Constant2Vector() {g=0.125, r=0.07}
  n83 = ScalarParameter() {defaultValue=0.1, group=MiddleClouds, parameterName=Bot_MCloudSpeedMult, sliderMax=1}
  n84 = TextureSample(UVs=Panner(
    Coordinate=MaterialFunctionCall(UVs=@n80, Texture Scale=@n81) {materialFunction=&ScaleUVsByCenter}.UVs
    Speed=Multiply(A=@n82, B=@n83)
   ) {speedX=-0.125, speedY=0.125}) {texture=&T_Clouds}
  NamedRerouteDeclaration(None=@n84.R) {name=M Cloud R Bot, nodeColor=<rgba 0 0 0 1>}
  NamedRerouteDeclaration(None=@n84.A) {name=M Cloud A Bot, nodeColor=<rgba 0 0 0 1>}
  NamedRerouteDeclaration(None=@n9.Scaled to Fill) {name=ScaleToFill, nodeColor=<rgba 1 0 0.8203 1>}
  n85 = NamedRerouteUsage()
  n86 = LinearInterpolate(
   A=ScalarParameter() {defaultValue=0.4, group=LightRays, parameterName=FocalMaskMin}
   B=Multiply(A=ScalarParameter() {defaultValue=1.5, group=LightRays, parameterName=FocalMaskMax}) {constB=2}
   Alpha=NamedRerouteUsage()
  ) {constA=1, constB=4}
  n87 = NamedRerouteUsage()
  n88 = DDX(Value=@n87)
  n89 = DDY(Value=@n87)
  n90 = Multiply(A=Time(), B=ScalarParameter() {defaultValue=0.01, group=LightRays, parameterName=LightRaysSpeed}) {constB=-0.25}
  n91 = MaterialFunctionCall!(
   UV=NamedRerouteUsage()
   Tiling=AppendVector(
    A=ScalarParameter() {defaultValue=4, group=LightRays, parameterName=LightRaysFrequency}
    B=ScalarParameter() {defaultValue=0.001, group=LightRays, parameterName=LightRaysStretch}
   )
   Offset=AppendVector(A=Constant(), B=Frac(None=@n90))
   Mask Size=Constant()
   Blend Distance=Constant() {r=1}
  ) {materialFunction=/FortUILibrary/MaterialFunctions/Effects/MF_UI_Funnel}
  n92 = Panner(Coordinate=@n91.Wrapped Funnel UV) {speedX=-0.01}
  n93 = TextureSample!(UVs=@n92, Tex, "DDX(UVs)"=@n88, "DDY(UVs)"=@n89, Apply View MipBias) {mipValueMode=TMVM_Derivative, samplerType=SAMPLERTYPE_Grayscale, texture=&LowResBlurredNoise}
  n94 = LinearInterpolate(
   A=TextureSample(UVs=@n92, "DDX(UVs)"=@n88, "DDY(UVs)"=@n89) {mipValueMode=TMVM_Derivative, texture=/Game/Global/Textures/Default/Noise/Noise}.R
   Alpha=@n93.R
  )
  n95 = ScalarParameter() {defaultValue=0.3, group=LightRays, parameterName=LightraysOpacity, sliderMax=1}
  n96 = Multiply(A=Divide(A=NamedRerouteUsage(), B=Multiply(A=@n86, B=@n94)), B=@n95) {constB=0.1}
  n97 = LinearInterpolate(
   A=ScalarParameter() {defaultValue=0.4, group=LightRays, parameterName=FocalMaskMin}
   B=ScalarParameter() {defaultValue=1.5, group=LightRays, parameterName=FocalMaskMax}
   Alpha=NamedRerouteUsage()
  ) {constA=1, constB=2}
  n98 = Panner(Coordinate=@n91.Wrapped Funnel UV) {speedX=0.01}
  n99 = TextureSample(UVs=@n98, "DDX(UVs)"=@n88, "DDY(UVs)"=@n89) {mipValueMode=TMVM_Derivative, samplerType=SAMPLERTYPE_Grayscale, texture=&LowResBlurredNoise}
  n100 = LinearInterpolate(
   A=TextureSample(UVs=@n98, "DDX(UVs)"=@n88, "DDY(UVs)"=@n89) {mipValueMode=TMVM_Derivative, texture=&Noise}.R
   Alpha=@n99.R
  )
  n101 = Multiply(A=@n95, B=Divide(A=NamedRerouteUsage(), B=Multiply(A=@n97, B=@n100))) {constB=0.1}
  n102 = MaterialFunctionCall(
   UVs=NamedRerouteUsage()
   Texture Scale=ScalarParameter() {defaultValue=3.75, group=LightRays, parameterName=LightRaysMaskScale, sliderMax=10}
  ) {materialFunction=&ScaleUVsByCenter}
  n103 = Multiply(
   A=Saturate(None=Add(A=@n96, B=@n101))
   B=MaterialFunctionCall("Value (0-1)"=MaterialFunctionCall(UV=@n102.UVs) {materialFunction=&MF_UI_RadialGradient}.Result) {materialFunction=&MF_UI_EaseCurves}.EaseInOutSine
  )
  NamedRerouteDeclaration(None=LinearInterpolate(A=LinearInterpolate(A=@n85, B=@n103), B=NamedRerouteUsage(), Alpha=NamedRerouteUsage())) {name=LightRays, nodeColor=<rgba 0 0 0 1>}
  n104 = NamedRerouteUsage()
  n105 = ScalarParameter() {defaultValue=1.5, group=MiddleClouds, parameterName=TopCloudMaskScale, sliderMax=10}
  NamedRerouteDeclaration(None=MaterialFunctionCall(UV=MaterialFunctionCall(UVs=@n104, Texture Scale=@n105) {materialFunction=&ScaleUVsByCenter}.UVs) {materialFunction=&MF_UI_RadialGradient}.Result) {name=TopCloudMask, nodeColor=<rgba 0 0 0 1>}
  n106 = Constant()
  n107 = ScalarParameter() {defaultValue=0.13, group=Background, parameterName=FocalPointHeight, sliderMax=0.5}
  NamedRerouteDeclaration(None=Add(A=@n21.Focal Point UV, B=AppendVector(A=@n106, B=@n107))) {name=Focal Point UV, nodeColor=<rgba 0 0 0 1>}
  NamedRerouteUsage()
  NamedRerouteUsage()
  NamedRerouteUsage()
  TextureCoordinate()
 ]
}
