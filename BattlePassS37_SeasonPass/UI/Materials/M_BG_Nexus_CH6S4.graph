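# M_BG_Nexus_CH6S4 is a masked background material for the Season 37 battle pass UI, built on shared front-end Vault World material functions.
# MF_BG_Nexus_CH6S4 generates the background color, which passes through a post-process calibration function.
# The calibrated color then feeds the Vault World transition function, so the material itself is only a thin chain of three function calls.
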Material /BattlePassS37_SeasonPass/UI/Materials/M_BG_Nexus_CH6S4
graph {
 properties {blendMode=BLEND_Masked, pixelDepthOffsetMode=PDOM_Legacy, usedShadingModels=Default Lit}
 nodes [
  n1 = MaterialFunctionCall!(Vault World Color=MaterialFunctionCall() {materialFunction=/BattlePassS37_SeasonPass/UI/MaterialFunctions/MF_BG_Nexus_CH6S4}.Result) {materialFunction=/Game/Environments/FrontEnd/Blueprints/ItemPreview/VaultWorldTransitions/MaterialFunctions/MF_UI_VaultWorld_PostProcessCalibration}
  MaterialFunctionCall!(Vault World Color=@n1.Calibrated Color) {materialFunction=/Game/Environments/FrontEnd/Blueprints/ItemPreview/VaultWorldTransitions/MaterialFunctions/MF_UI_VaultWorld_Transition}
 ]
}
