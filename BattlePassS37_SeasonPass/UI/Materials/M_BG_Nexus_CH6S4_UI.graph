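# M_BG_Nexus_CH6S4_UI is a translucent UI-domain material in the Season 37 battle pass plugin, apparently a Nexus background for Chapter 6 Season 4.
# Its graph holds a single call to the material function MF_BG_Nexus_CH6S4, so the material acts as a thin UI wrapper with the visual logic kept in that function.
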Material /BattlePassS37_SeasonPass/UI/Materials/M_BG_Nexus_CH6S4_UI
graph {
 properties {blendMode=BLEND_Translucent, materialDomain=MD_UI, pixelDepthOffsetMode=PDOM_Legacy, usedShadingModels=Default Lit}
 nodes [MaterialFunctionCall() {materialFunction=/BattlePassS37_SeasonPass/UI/MaterialFunctions/MF_BG_Nexus_CH6S4}]
}
