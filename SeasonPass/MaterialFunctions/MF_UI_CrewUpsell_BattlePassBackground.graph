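# UI material function for the Crew upsell's Battle Pass background, outputting Color and an angled-cut Opacity.
# It composites an offset, scaled Image texture over the Background Color input, masked by alpha, edge clipping and a bottom fade.
# An eased vertical gradient then blends toward Fade Color.
# Two offset BoxAngle masks subtract into a division line over the color, while DisplayCut blends the cut into Opacity.
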MaterialFunction /SeasonPass/MaterialFunctions/MF_UI_CrewUpsell_BattlePassBackground
graph {
 properties {
  bExposeToLibrary=true
  description="Merge the background, the image and the fading together. It also supports having the angled cut as a mask."
  previewMaterialDomain=MD_UI
  libraryCategoriesText [Battle Pass]
 }
 nodes [
  n1 = AppendVector(A=ScalarParameter() {defaultValue=-0.025, group=Upsell Settings, parameterName=ImageOffsetX}, B=Constant() {r=-0.03}) {desc=Image Offset}
  n2 = TextureObjectParameter() {group=Upsell Settings, parameterName=Image, texture=/Game/Global/Textures/Default/Blanks/T_Blank_White_ZeroAlpha_sRGBColor}
  n3 = MaterialFunctionCall!(Use Local UV, Texture Size=TextureProperty(Tex=@n2), Scale Pivot Point=Constant2Vector() {r=0.5}) {materialFunction=/Game/Environments/FrontEnd/Blueprints/ItemPreview/VaultWorldTransitions/MaterialFunctions/MF_UI_VaultWorld_UV}
  n4 = Add(A=@n1, B=@n3.Scale X)
  n5 = Constant2Vector() {r=0.5}
  n6 = Add(A=Multiply(A=Subtract(A=@n4, B=@n5) {constB=0.5}) {constB=1.4, desc=Image Scale}, B=@n5) {constB=0.5}
  n7 = TextureSample(UVs=@n6, Tex=@n2)
  n8 = NamedRerouteUsage()
  n9 = LinearInterpolate(
   A=FunctionInput() {description=Set your backgorund visuals, inputName=Background Color}
   B=@n7.RGB
   Alpha=MaterialFunctionCall!(UVs=@n6, Alpha=Multiply(A=@n8, B=@n7.A)) {materialFunction=/FortUILibrary/MaterialFunctions/Basics/MF_UI_ClipEdgePixel}.Result
  )
  n10 = FunctionInput() {description=Set the fading color, inputName=Fade Color, sortPriority=10}
  n11 = ComponentMask(None=NamedRerouteUsage()) {g=true}
  n12 = ScalarParameter() {defaultValue=0.275, group=Upsell Settings, parameterName=BtmGradInputLow, sliderMax=1}
  n13 = ScalarParameter() {defaultValue=0.8, group=Upsell Settings, parameterName=BtmGradInputHigh, sliderMax=1}
  n14 = MaterialFunctionCall("Value (0-1)"=Saturate(None=MaterialFunctionCall(Input=@n11, Input Low=@n12, Input High=@n13) {materialFunction=&RemapValue_0-1}.Result)) {desc=Color Gradient Range, materialFunction=&MF_UI_EaseCurves}
  FunctionOutput(None=LinearInterpolate(A=LinearInterpolate(A=@n9, B=@n10, Alpha=@n14.EaseOutCubic), Alpha=NamedRerouteUsage())) {description=Ouput your background with the image that is faded, outputName=Color}
  n15 = TextureCoordinate()
  n16 = MaterialFunctionCall(UV=LinearInterpolate(A=OneMinus(None=@n15), B=@n15, Alpha=Constant() {r=1}), ScalePivotPoint=Constant2Vector()) {materialFunction=&MF_UI_UVScaling}
  n17 = MaterialFunctionCall!(UV=@n16.ScaleY, Corner=Constant() {r=3}, Angle=Constant() {r=4}, Flip, Falloff) {materialFunction=/FortUILibrary/MaterialFunctions/Shapes/MF_UI_BoxAngle}
  FunctionOutput(None=LinearInterpolate(B=@n17.Result, Alpha=ScalarParameter() {defaultValue=1, parameterName=DisplayCut, sliderMax=1}) {constA=1}) {description=Ouput the angle cut for the opacity, outputName=Opacity, sortPriority=10}
  NamedRerouteDeclaration(None=Add(A=@n1, B=@n3.UV)) {name=Base UV, nodeColor=<rgba 0 0 0 1>}
  n18 = MaterialFunctionCall(Input=Saturate(None=ComponentMask(None=@n6) {g=true}), Input Low=Constant() {r=0.9}, Input High=Constant() {r=1}) {materialFunction=&RemapValue_0-1}
  NamedRerouteDeclaration(None=Saturate(None=OneMinus(None=@n18.Result))) {name=Image Fade, nodeColor=<rgba 0 0 0 1>}
  n19 = MaterialFunctionCall(UV=Add(A=@n16.ScaleY, B=Constant2Vector() {desc=Division Line Offset, r=-0.015}) {constB=0}, Corner=Constant() {r=3}, Angle=Constant() {r=6}) {materialFunction=&MF_UI_BoxAngle}
  NamedRerouteDeclaration(None=Saturate(None=Subtract(A=@n17.Result, B=@n19.Result))) {name=Division Line, nodeColor=<rgba 0 0 0 1>}
 ]
}
