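# Translucent UI material for the Crew upsell background in the Season 37 (CH6S4) Battle Pass screens.
# The season's Nexus background function, MF_BG_Nexus_CH6S4, supplies the background color to the shared SeasonPass function MF_UI_CrewUpsell_BattlePassBackground.
# An UpsellFadeColor vector parameter, defaulting to dark blue, feeds that function's fade color.
# The material holds only this wiring, with the visual logic living in the two functions.
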Material /BattlePassS37_SeasonPass/UI/Materials/M_UI_Crew_Upsell_CH6S4
graph {
 properties {bAutomaticallySetUsageInEditor=false, blendMode=BLEND_Translucent, materialDomain=MD_UI, usedShadingModels=Default Lit}
 nodes [
  n1 = VectorParameter() {desc=Color Theme 04, group=Upsell, parameterName=UpsellFadeColor, defaultValue=<rgba 0.0395 0.0762 0.2195 1>}
  MaterialFunctionCall!(
   Background Color=MaterialFunctionCall() {materialFunction=/BattlePassS37_SeasonPass/UI/MaterialFunctions/MF_BG_Nexus_CH6S4}.Result
   Fade Color=@n1.RGB
  ) {materialFunction=/SeasonPass/MaterialFunctions/MF_UI_CrewUpsell_BattlePassBackground}
 ]
}
